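; Data asset for Mojo stage screen events, supplying screen materials and Niagara effects by name key.
; Material lookups split slot names at the last underscore, using the suffix to pick Squared, Horizontal, Vertical or default materials.
; GetMaterialsAndSlotsByType filters for "_Screen_" slots, with fallbacks to legacy or first-entry materials.
; It and GetScreenNiagara resolve soft references without loading; GetMaterialsAndSlots and GetScreenMeshesLegacyValues are empty stubs.
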(fn GetFilteredMaterialsByType
  (params (SlotNames :type Array%20of%20Names))
  (bind _materialslotsmap (Variables|Default|GetMaterialSlotsMap))
  ;[1 1 1] Grab the map using the screen key
  (LocalVariables|SetLocalScreenType SlotNames)
  (Utilities|Array|Clear (LocalVariables|GetLocalLoadedMaterialsList))
  ;[1 1 1] Find all materials based on slot name key and send it.
  (for _array_element (LocalVariables|GetLocalScreenType)
    ;;[1 1 1] Try to grab the material by the type (squared, Vertical, Horizontal) if the material id has the information.
    (if (.ReturnValue (Utilities|Map|Find _materialslotsmap (Utilities|String|StringToName (Utilities|String|Split (Utilities|String|ToString(Name) _array_element) "_" "IgnoreCase" "FromEnd"))))
      (bind _return (CallFunction|GetScreenMaterialType (Utilities|Map|Find _materialslotsmap (Utilities|String|StringToName (Utilities|String|Split (Utilities|String|ToString(Name) _array_element) "_" "IgnoreCase" "FromEnd"))) (Utilities|String|StringToName (.RightS (Utilities|String|Split (Utilities|String|ToString(Name) _array_element) "_" "IgnoreCase" "FromEnd")))))
      (Utilities|Array|Add (LocalVariables|GetLocalLoadedMaterialsList) _return)))
  (return (LocalVariables|GetLocalLoadedMaterialsList) (Utilities|Array|IsNotEmpty (LocalVariables|GetLocalLoadedMaterialsList))))

(fn GetMaterialsAndSlots
  (params (ScreenStageKey :type Name :default "None"))
  (return 0 0 false))

(fn GetMaterialsAndSlotsByType
  (params (MaterialSlotNames :type Array%20of%20Names))
  (bind _currentslotsandmaterialsmap (LocalVariables|GetCurrentSlotsAndMaterialsMap))
  (bind _materialslotsmap (Variables|Default|GetMaterialSlotsMap))
  (LocalVariables|SetLocalScreenType MaterialSlotNames)
  ;[1 1 1] Grab the map using the screen key
  (Utilities|Array|Clear (LocalVariables|GetLocalLoadedMaterialsList))
  ;[1 1 1] Find all materials based on slot name key and send it.
  (Utilities|Array|Clear (LocalVariables|GetLocalKeysFiltered))
  (for _array_element (LocalVariables|GetLocalScreenType)
    (bind _values (Utilities|Map|Values (Variables|Default|GetMaterialSlotsMap)))
    (if (and (Utilities|Array|IsNotEmpty _values) (Utilities|String|Contains (Utilities|String|ToString(Name) _array_element) "_Screen_"))
      ;;[0 0.2537 0.7031] Try to return the material based on the type (squared, horizontal, vertical)
      (if (.ReturnValue (Utilities|Map|Find _materialslotsmap (Utilities|String|StringToName (Utilities|String|Split (Utilities|String|ToString(Name) _array_element) "_" "IgnoreCase" "FromEnd"))))
        (bind _return (CallFunction|GetScreenMaterialType (Utilities|Map|Find _materialslotsmap (Utilities|String|StringToName (Utilities|String|Split (Utilities|String|ToString(Name) _array_element) "_" "IgnoreCase" "FromEnd"))) (Utilities|String|StringToName (.RightS (Utilities|String|Split (Utilities|String|ToString(Name) _array_element) "_" "IgnoreCase" "FromEnd")))))
        (if (Utilities|IsValidSoftObjectReference _return)
          (Utilities|IsValid (Utilities|ResolveSoftReference _return)
            (:Is%20Valid
              (Utilities|Array|Add (LocalVariables|GetLocalLoadedMaterialsList) (Utilities|ResolveSoftReference _return))))
          (else
            ;;;[1 0 0.05781] If for some reason there is no valid material configured try to get the legacy one as fallback.
            (bind _return_2 (CallFunction|GetScreenMaterialType (Utilities|Map|Find _materialslotsmap (Utilities|String|StringToName (Utilities|String|Split (Utilities|String|ToString(Name) _array_element) "_" "IgnoreCase" "FromEnd")))))
            (Utilities|IsValid (Utilities|ResolveSoftReference _return_2)
              (:Is%20Valid
                (Utilities|Array|Add (LocalVariables|GetLocalLoadedMaterialsList) (Utilities|ResolveSoftReference _return_2))))))
        (Utilities|Array|Add (LocalVariables|GetLocalKeysFiltered) _array_element)
        (elif (.ReturnValue (Utilities|Map|Find _currentslotsandmaterialsmap _array_element))
          ;;[0.7031 0.6338 0] If it fail to find the type tag on the material id just return the legacy option.
          (bind _return_3 (CallFunction|GetScreenMaterialType (Utilities|Map|Find _currentslotsandmaterialsmap _array_element)))
          (Utilities|IsValid (Utilities|ResolveSoftReference _return_3)
            (:Is%20Valid
              (Utilities|Array|Add (LocalVariables|GetLocalLoadedMaterialsList) (Utilities|ResolveSoftReference _return_3))
              (Utilities|Array|Add (LocalVariables|GetLocalKeysFiltered) _array_element)))
          (else
            ;;[1 0.216 0.988] In case there is no valid material slot name on the map fallback to the first slot name on the map if the screen has a Screen key on its material slot name.
            (bind _values_2 (Utilities|Map|Values (LocalVariables|GetCurrentSlotsAndMaterialsMap)))
            (bind _return_4 (CallFunction|GetScreenMaterialType (Utilities|Array|Get(acopy) _values_2) (Utilities|String|StringToName (.RightS (Utilities|String|Split (Utilities|String|ToString(Name) _array_element) "_" "IgnoreCase" "FromEnd")))))
            (if (Utilities|IsValidSoftObjectReference _return_4)
              (Utilities|IsValid (Utilities|ResolveSoftReference _return_4)
                (:Is%20Valid
                  (Utilities|Array|Add (LocalVariables|GetLocalLoadedMaterialsList) (Utilities|ResolveSoftReference _return_4))))
              (else
                ;;;[1 0 0.05781] If for some reason there is no valid material configured try to get the legacy one as fallback.
                (bind _return_5 (CallFunction|GetScreenMaterialType (Utilities|Array|Get(acopy) _values_2)))
                (Utilities|IsValid (Utilities|ResolveSoftReference _return_5)
                  (:Is%20Valid
                    (Utilities|Array|Add (LocalVariables|GetLocalLoadedMaterialsList) (Utilities|ResolveSoftReference _return_5))))))
            (Utilities|Array|Add (LocalVariables|GetLocalKeysFiltered) _array_element))))))
  (return (LocalVariables|GetLocalKeysFiltered) (LocalVariables|GetLocalLoadedMaterialsList) (Utilities|Array|IsNotEmpty (LocalVariables|GetLocalLoadedMaterialsList))))

(fn GetScreenMaterialType
  (params (Materials :type S%20Mojo%20Stage%20Screen%20Material%20Type%20Structure) (ScreenType :type Name :default "None"))
  (if (== ScreenType "Squared")
    (return (Utilities|Struct|BreakSMojoStageScreenMaterialType Materials))
    (elif (== ScreenType "Horizontal")
      (return (Utilities|Struct|BreakSMojoStageScreenMaterialType Materials))
      (elif (== ScreenType "Vertical")
        (return (Utilities|Struct|BreakSMojoStageScreenMaterialType Materials))
        (else
          (return (Utilities|Struct|BreakSMojoStageScreenMaterialType Materials)))))))

(fn GetScreenMeshesKey
  (bind _keys (Utilities|Map|Keys (Variables|Default|GetMaterialSlotsMap)))
  (return _keys))

(fn GetScreenMeshesLegacyValues
  (return 0))

(fn GetScreenMeshesValues
  (bind _values (Utilities|Map|Values (Variables|Default|GetMaterialSlotsMap)))
  (return _values))

(fn GetScreenNiagara
  (params (Key :type Name :default "None"))
  (bind _screenniagaras (Variables|Default|GetScreenNiagaras))
  (bind (_value _returnvalue) (Utilities|Map|Find _screenniagaras Key))
  (bind _value_niagaraarray_13_65ec2fb74b24c8080098fbb073391cdb (Utilities|Struct|BreakSMojoStageScreenNiagaras _value))
  (Utilities|Array|Clear (LocalVariables|GetLocalScreenNiagaras))
  (if _returnvalue
    (for _array_element _value_niagaraarray_13_65ec2fb74b24c8080098fbb073391cdb
      (Utilities|IsValid (Utilities|ResolveSoftReference _array_element)
        (:Is%20Valid
          (Utilities|Array|Add (LocalVariables|GetLocalScreenNiagaras) (Utilities|ResolveSoftReference _array_element)))))
    (return (LocalVariables|GetLocalScreenNiagaras))))

(fn GetScreenNiagaraKeys
  (bind _keys (Utilities|Map|Keys (Variables|Default|GetScreenNiagaras)))
  (return _keys))

(fn GetScreenNiagaraValues
  (bind _values (Utilities|Map|Values (Variables|Default|GetScreenNiagaras)))
  (return _values))
