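; BP_MojoStageScreenManager drives stage screen visuals in Pilgrim Quickplay, registering with the playspace's StageShowManager on cosmetic-capable instances.
; MIDI or gameplay-state events map through data assets to screen events, which PlayEvent pushes as materials and Niagara to screen actors via interface.
; Legacy venues match materials by screen key, others by slot type; an async loader preloads them.
; It also syncs a screen brightness parameter to Lumen and supports debug-forced events.
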(event Custom|ActivateDebugMode
  (params (Active :type Boolean))
  ;[1 0.2833 0] For the console command and disable all MIDI or quickplay event listener
  (Variables|Default|SetDebugMode Active))

(event EventBeginPlay
  (Networking|CanExecuteCosmeticEvents
    (:True
      (Fortnite|FortShippingLog self "[StageScreenManager] Initializing.")
      ;[1 1 1] Set screen brightness for non lumen
      (EventDispatchers|BindEventtoOnGlobalIlluminationQualityChanged (Settings|GetFortGameUserSettings) (delegate Custom|HandleSettingChanged_GI))
      (CallFunction|HandleSettingChangedGI)
      ; Register Screen Manager on Stage Show Manager Component, if present
      (bind _eventrouter (Pilgrim|PilgrimWaitForEventRouter 0)
        (:OnEventRouterAvailable
          (bind _output_get (Variables|Default|SetEventRouter _eventrouter))
          (bind _assparks_music_playspace (Utilities|Casting|CastToSparksMusicPlayspace (Components|GetOwner _output_get)))
          (Utilities|IsValid (Actor|GetComponentbyClass _assparks_music_playspace "/PilgrimQuickplay/StageManager/StageShowManager/BP_PilgrimQuickplay_StageShowManager.BP_PilgrimQuickplay_StageShowManager_C")
            (:Is%20Valid
              (Class|BPPilgrimQuickplayStageShowManager|InitializeScreenManager (Actor|GetComponentbyClass _assparks_music_playspace "/PilgrimQuickplay/StageManager/StageShowManager/BP_PilgrimQuickplay_StageShowManager.BP_PilgrimQuickplay_StageShowManager_C") self))))))))

(event Custom|HandleSettingChanged_GI
  (bind _isenabled (Class|BTechArtGlobalLibrary|IsLumenEnabled))
  ;[1 1 1] Set screen brightness for non lumen
  (Rendering|Material|SetScalarParameterValue "/SparksEffects/Effects/MPC_Sparks.MPC_Sparks" "ScreensHighGISetting" (select _isenabled 1.0 0.0) :!class /Script/Engine.KismetMaterialLibrary)
  (Fortnite|FortShippingLog self (Utilities|String|Append "Screen Manager Is Lumen Enabled =" (Utilities|String|ToString(Boolean) _isenabled))))

(event EventEndPlay
  (params (EndPlayReason :type EEndPlayReason%20Enum :default "Destroyed"))
  ;[1 1 1] Set screen brightness for non lumen
  (EventDispatchers|UnbindEventfromOnGlobalIlluminationQualityChanged (Settings|GetFortGameUserSettings) (delegate Custom|HandleSettingChanged_GI)))

(event Custom|LoadAllMaterialsAndNiagaras
  (params (ScreenEvents :type Array%20of%20Mojo%20Stage%20Screen%20Event%20Interface%20Interfaces))
  (bind _currenteventindex (Variables|LoadAsync|GetCurrentEventIndex))
  (bind _output (Utilities|Array|Get(acopy) ScreenEvents _currenteventindex))
  (bind _currentscreenniagaraindex (Variables|LoadAsync|GetCurrentScreenNiagaraIndex))
  (bind _filteredmaterialstoload (Variables|Default|GetFilteredMaterialsToLoad))
  ;[0.8194 0.197 0.05956] Reset Loop Values
  (Variables|LoadAsync|SetCurrentEventIndex 0)
  (Variables|LoadAsync|SetCurrentScreenMeshIndex 0)
  (Variables|LoadAsync|SetCurrentSlotMaterialsMapIndex 0)
  ;[0.125 1 0.209] Filtered Event Loop
  (if (< (Variables|LoadAsync|GetCurrentEventIndex) (Utilities|Array|Length ScreenEvents))
    (Utilities|FlowControl|Sequence
      (:then_0
        _)
      (:then_1
        (bind _values (Class|MojoStageScreenEventInterface|GetScreenMeshesValues _output))
        (bind _legacy (Class|MojoStageScreenEventInterface|GetScreenMeshesLegacyValues _output))
        (Variables|LoadAsync|SetCurrentScreenMeshIndex 0)
        ;[0.1567 0.2051 0.8594] Current Screen Mesh Loop
        (if (< (Variables|LoadAsync|GetCurrentScreenMeshIndex) (Math|Integer|SelectInt (Utilities|Array|Length _legacy) (Utilities|Array|Length _values) (Variables|Default|GetIsLegacyVenue)))
          (if (Variables|Default|GetIsLegacyVenue)
            (bind _output_slotsandmaterialsmap_9_65ec2fb74b24c8080098fbb073391cdb (Utilities|Struct|BreakSMojoStageScreenSlotsAndMaterialsLegacy (Utilities|Array|Get(acopy) _legacy (Variables|LoadAsync|GetCurrentScreenMeshIndex))))
            (bind _values_2 (Utilities|Map|Values _output_slotsandmaterialsmap_9_65ec2fb74b24c8080098fbb073391cdb))
            (Variables|Default|SetFilteredMaterialsToLoad _values_2))
          (Variables|LoadAsync|SetCurrentSlotMaterialsMapIndex 0)
          ;[0.9236 0.08362 0.8949] Loading Async Loop
          (if (< (Variables|LoadAsync|GetCurrentSlotMaterialsMapIndex) (Utilities|Array|Length _filteredmaterialstoload))
            (bind _object (Utilities|AsyncLoadAsset (Utilities|Array|Get(acopy) _filteredmaterialstoload (Variables|LoadAsync|GetCurrentSlotMaterialsMapIndex)))
              (:Completed
                (Utilities|IsValid _object
                  (:Is%20Valid
                    (bind _returnvalue (Utilities|Array|AddUnique (Variables|LoadAsync|GetAllLoadedObjects) _object))
                    (if (Variables|Default|GetEnableDebugLogs)
                      (if (>= _returnvalue 0)
                        (Development|PrintText (Utilities|Text|FormatText "NSLOCTEXT(\"[B4076FD6FD202517AE1D218B8F57CF4B]\", \"2D3ED0AB42F40F3422D468BE5A0F5B21\", \"[StageScreenManager] {index} Loaded asset: {name}\")" _returnvalue (Utilities|GetDisplayName _object)) false)
                        (Math|Integer|IncrementInt (Variables|LoadAsync|GetCurrentSlotMaterialsMapIndex)
                          (:%20%20
                            _)))))
                  (:Is%20Not%20Valid
                    _)))))
          (Math|Integer|IncrementInt (Variables|LoadAsync|GetCurrentScreenMeshIndex)
            (:%20%20
              _)))
        (Math|Integer|IncrementInt (Variables|LoadAsync|GetCurrentEventIndex)
          (:%20%20
            _))))
    (if (not (Variables|Default|GetIsLegacyVenue))
      ;[0.7031 0.6338 0] Filter the material that will be loaded based on the type (legacy, vertical, horizontal, squared)
      (Utilities|Array|Clear (Variables|Default|GetFilteredMaterialsToLoad))
      (bind _values_3 (Utilities|Map|Values (Variables|Default|GetScreenStageActors)))
      (for _array_element _values_3
        (bind _materialslots (Class|MojoStageScreenShowFxInterface|GetMaterialsSlotNames(Message) _array_element))
        (bind (_materials _isvalid) (Class|MojoStageScreenEventInterface|GetFilteredMaterialsbyType (Utilities|Array|Get(acopy) ScreenEvents (Variables|LoadAsync|GetCurrentEventIndex)) _materialslots))
        (if _isvalid
          (for _array_element_2 _materials
            (Utilities|Array|Add (Variables|Default|GetFilteredMaterialsToLoad) _array_element_2))))))
  ;[0.8194 0.276 0.05956] Reset Loop Values
  (Variables|LoadAsync|SetCurrentEventIndex 0)
  (Variables|LoadAsync|SetCurrentNiagaraIndex 0)
  (Variables|LoadAsync|SetCurrentScreenNiagaraIndex 0)
  ;[0.125 1 0.209] Filtered Event Loop
  (if (< (Variables|LoadAsync|GetCurrentEventIndex) (Utilities|Array|Length ScreenEvents))
    (Variables|LoadAsync|SetCurrentScreenNiagaraIndex 0)
    (bind _values_4 (Class|MojoStageScreenEventInterface|GetScreenNiagaraValues (Utilities|Array|Get(acopy) ScreenEvents (Variables|LoadAsync|GetCurrentEventIndex))))
    ;[0.1567 0.2051 0.8594] CurrentScreenNiagaraIndex
    (if (< (Variables|LoadAsync|GetCurrentScreenNiagaraIndex) (Utilities|Array|Length _values_4))
      (Variables|LoadAsync|SetCurrentNiagaraIndex 0)
      ;[0.9236 0.08362 0.8949] Loading Async Loop
      (bind _output_niagaraarray_13_65ec2fb74b24c8080098fbb073391cdb (Utilities|Struct|BreakSMojoStageScreenNiagaras (Utilities|Array|Get(acopy) _values_4 _currentscreenniagaraindex)))
      (if (< (Variables|LoadAsync|GetCurrentNiagaraIndex) (Utilities|Array|Length _output_niagaraarray_13_65ec2fb74b24c8080098fbb073391cdb))
        (bind _object_2 (Utilities|AsyncLoadAsset (Utilities|Array|Get(acopy) _output_niagaraarray_13_65ec2fb74b24c8080098fbb073391cdb (Variables|LoadAsync|GetCurrentNiagaraIndex)))
          (:Completed
            (Utilities|IsValid _object_2
              (:Is%20Valid
                (bind _returnvalue_2 (Utilities|Array|AddUnique (Variables|LoadAsync|GetAllLoadedObjects) _object_2))
                (if (Variables|Default|GetEnableDebugLogs)
                  (if (>= _returnvalue_2 0)
                    (Development|PrintText (Utilities|Text|FormatText "NSLOCTEXT(\"[B4076FD6FD202517AE1D218B8F57CF4B]\", \"762F7E744AA3A0AA465F03B415A3CDED\", \"[StageScreenManager] Loaded asset {index}: {name}\")" _returnvalue_2 (Utilities|GetDisplayName _object_2)))
                    (Math|Integer|IncrementInt (Variables|LoadAsync|GetCurrentNiagaraIndex)
                      (:%20%20
                        _)))))
              (:Is%20Not%20Valid
                _)))))
      (Math|Integer|IncrementInt (Variables|LoadAsync|GetCurrentScreenNiagaraIndex)
        (:%20%20
          _)))
    (Math|Integer|IncrementInt (Variables|LoadAsync|GetCurrentEventIndex)
      (:%20%20
        _)))
  (Development|PrintString "[StageScreenManager] Finished loading assets" false)
  (Default|CallOnFinishedLoadingAssets))

(fn ForceEvent
  (params (ThemeName :type String :default "EyeScreen") (EventName :type String :default "DA_StageScreenEvent_Chorus_Eye"))
  (bind _outassetdata (AssetRegistry|GetAssetsbyPath (AssetRegistry|GetAssetRegistry) (Utilities|String|StringToName (Utilities|String|Append "/SparksCommon/Systems/StageManager/ScreenManager/Data/" ThemeName)) false false))
  ;[0.9896 0.9012 0.09277] Check if the asset is what we want to play comparing with the given name from the console command.
  (for _array_element _outassetdata
    (if (Utilities|String|Contains (AssetData|GetFullName _array_element) EventName false true)
      ;[0.125 1 0.209] If find it play it!
      (bind _asmojo_stage_screen_event_interface (Utilities|Casting|CastToMojoStageScreenEventInterface (AssetData|GetAsset _array_element)))
      (CallFunction|PlayEvent _asmojo_stage_screen_event_interface true))))

(fn GrabAllStageScreenActors
  (Utilities|Map|Clear (Variables|Default|GetScreenStageActors))
  (bind _outactors (Actor|GetAllActorswithInterface "/SparksStageShow/Systems/StageManager/ScreenManager/MojoStageScreenShowFxInterface.MojoStageScreenShowFxInterface_C"))
  (for _array_element _outactors
    (if (== (Level|GetLevel _array_element) (Level|GetLevel))
      (bind _return (Class|MojoStageScreenShowFxInterface|GetScreenID(Message) _array_element))
      (Utilities|Map|Add (Variables|Default|GetScreenStageActors) _return _array_element))))

(fn PlayEvent
  (params (ScreenEvent :type Mojo%20Stage%20Screen%20Event%20Interface%20Interface) (isDebug :type Boolean))
  (bind _screenstageactors (Variables|Default|GetScreenStageActors))
  (bind _screenstageactors_2 (Variables|Default|GetScreenStageActors))
  (bind _output_get (LocalVariables|SetLocalCurrentEvent ScreenEvent))
  (if (Utilities|IsValidInterface _output_get)
    ;[1 1 1] If is same event do nothing
    (Utilities|IsValid (Utilities|ToObject(Interface) (Variables|Default|GetLastEvent))
      (:Is%20Valid
        (if (or (!= (Utilities|ToObject(Interface) (LocalVariables|GetLocalCurrentEvent)) (Utilities|ToObject(Interface) (Variables|Default|GetLastEvent))) isDebug)
          ;[1 1 1] Change the Screen Meshes Materials
          (bind _keys (Class|MojoStageScreenEventInterface|GetScreenMeshesKey (LocalVariables|GetLocalCurrentEvent)))))
      (:Is%20Not%20Valid
        _))
    (if (Variables|Default|GetIsLegacyVenue)
      (for _array_element _keys
        (bind (_slotnames _materials _isvalid) (Class|MojoStageScreenEventInterface|GetMaterialsandSlots (LocalVariables|GetLocalCurrentEvent) _array_element))
        (if (and _isvalid (.ReturnValue (Utilities|Map|Find _screenstageactors_2 _array_element)))
          (Class|MojoStageScreenShowFxInterface|UpdateMaterial(Message) (Utilities|Map|Find _screenstageactors_2 _array_element) _slotnames _materials)))
      (else
        ;[1 1 1] Run the 2.0 logic
        (bind _values (Utilities|Map|Values (Variables|Default|GetScreenStageActors)))
        (for _array_element_2 _values
          (bind _materialslots (Class|MojoStageScreenShowFxInterface|GetMaterialsSlotNames(Message) _array_element_2))
          (bind (_slotnames_2 _materials_2 _isvalid_2) (Class|MojoStageScreenEventInterface|GetMaterialsandSlotsbyType (LocalVariables|GetLocalCurrentEvent) _materialslots))
          (if _isvalid_2
            (Class|MojoStageScreenShowFxInterface|UpdateMaterial(Message) _array_element_2 _slotnames_2 _materials_2)))))
    ;[1 1 1] Change the Screen Niagara
    (bind _keys_2 (Class|MojoStageScreenEventInterface|GetScreenNiagaraKeys (LocalVariables|GetLocalCurrentEvent)))
    (for _array_element_3 _keys_2
      (bind _screenniagaras (Class|MojoStageScreenEventInterface|GetScreenNiagara (LocalVariables|GetLocalCurrentEvent) _array_element_3))
      (if (.ReturnValue (Utilities|Map|Find _screenstageactors _array_element_3))
        (if (Utilities|Array|IsNotEmpty _screenniagaras)
          (for _array_element_4 _screenniagaras
            (Class|MojoStageScreenShowFxInterface|UpdateNiagara(Message) (Utilities|Map|Find _screenstageactors _array_element_3) _array_element_4))
          (else
            (Class|MojoStageScreenShowFxInterface|UpdateNiagara(Message) (Utilities|Map|Find _screenstageactors _array_element_3) 0)))))
    (bind _output_get_2 (Variables|Default|SetLastEvent (LocalVariables|GetLocalCurrentEvent)))
    (if (Variables|Default|GetEnableDebugLogs)
      (Development|PrintString (Utilities|String|Append "[StageScreenManager] Played screen event: " (Utilities|GetDisplayName (Utilities|ToObject(Interface) _output_get_2))) true true "(R=0.000000,G=0.660000,B=1.000000,A=1.000000)" 1e+01 "StageScreenEvent"))
    (elif (Variables|Default|GetEnableDebugLogs)
      (Fortnite|FortShippingLog self "[StageScreenManager] Attempted to play invalid screen event." true))))

(fn PlayGameplayEvent
  (params (Event :type EPilgrimQuickplayState%20Enum :default "Invalid"))
  (bind _currentgameplayeventmap (Variables|Default|GetCurrentGameplayEventMap))
  (bind _eventmap (Variables|/PilgrimQuickplay/StageManager/ScreenManager/BP_MojoStageScreenGameplayMap_DataAsset.BP_MojoStageScreenGameplayMap_DataAsset_C|GetEventMap _currentgameplayeventmap))
  (bind (_value _returnvalue) (Utilities|Map|Find _eventmap Event))
  (if (not (Variables|Default|GetDebugMode))
    (if _returnvalue
      (CallFunction|PlayEvent _value (Variables|Default|GetDebugMode))
      (else
        (CallFunction|PlayEvent (Variables|/PilgrimQuickplay/StageManager/ScreenManager/BP_MojoStageScreenGameplayMap_DataAsset.BP_MojoStageScreenGameplayMap_DataAsset_C|GetDefaultScreenEvent _currentgameplayeventmap) (Variables|Default|GetDebugMode))))))

(fn PlayMIDIEvent
  (params (Event :type ESparksStageManager_Events_Midi_Text_Events%20Enum :default "NewEnumerator0"))
  (bind (_value _returnvalue) (Utilities|Map|Find (Variables|/PilgrimQuickplay/StageManager/ScreenManager/BP_MojoStageScreenMIDIMap_DataAsset.BP_MojoStageScreenMIDIMap_DataAsset_C|GetEventMap (Variables|Default|GetCurrentMIDIEventMap)) Event))
  (CallFunction|PlayEvent (select _returnvalue _value (Variables|/PilgrimQuickplay/StageManager/ScreenManager/BP_MojoStageScreenMIDIMap_DataAsset.BP_MojoStageScreenMIDIMap_DataAsset_C|GetDefaultScreenEvent (Variables|Default|GetCurrentMIDIEventMap)))))

(fn SetMIDIEventsMap
  (params (EventMap :type BP%20Mojo%20Stage%20Screen%20MIDIMap%20Data%20Asset%20Object%20Reference%20%28by%20ref%29))
  (bind _output_get (Variables|Default|SetMIDIEventMap EventMap))
  (bind _output_get_2 (Variables|Default|SetCurrentMIDIEventMap _output_get))
  (if (Variables|Default|GetEnableDebugLogs)
    (Development|PrintString (Utilities|String|Append "[StageScreenManager] Screen map set: " (Utilities|GetDisplayName _output_get_2)) true true "(R=0.000000,G=0.660000,B=1.000000,A=1.000000)" 1e+01)))

(fn SetGameplayEventsMap
  (params (Event%20Map :type BP%20Mojo%20Stage%20Screen%20Gameplay%20Map%20Data%20Asset%20Object%20Reference%20%28by%20ref%29))
  (bind _output_get (Variables|Default|SetGameplayEventMap Event%20Map))
  (bind _output_get_2 (Variables|Default|SetCurrentGameplayEventMap _output_get))
  (if (Variables|Default|GetEnableDebugLogs)
    (Development|PrintString (Utilities|String|Append "[StageScreenManager] Gameplay event map set:" (Utilities|GetDisplayName _output_get_2)) true true "(R=0.000000,G=0.660000,B=1.000000,A=1.000000)" 1e+01)))

(fn ConstructionScript)
